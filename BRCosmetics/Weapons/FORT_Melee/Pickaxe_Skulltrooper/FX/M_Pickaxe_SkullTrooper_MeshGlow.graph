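# Translucent unlit glow material for Niagara mesh particles, part of the Skulltrooper pickaxe effects.
# Particle color feeds MF_FX_EmissiveAdjustments for emissive color.
# A separate mask multiplies particle alpha, Fresnel and a falloff along U with scrolling trail noise reused from the LaserSword effects.
# The noise UVs are tiled and power-stretched along U before panning with a time offset, and two samples share them, summed and square-rooted.
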Material /BRCosmetics/Weapons/FORT_Melee/Pickaxe_Skulltrooper/FX/M_Pickaxe_SkullTrooper_MeshGlow
graph {
 properties {
  bUsedWithNiagaraMeshParticles=true
  blendMode=BLEND_Translucent
  pixelDepthOffsetMode=PDOM_Legacy
  shadingModel=MSM_Unlit
  usedShadingModels=Unlit
 }
 nodes [
  n1 = ParticleColor()
  MaterialFunctionCall(Emissive=@n1.RGB) {materialFunction=&MF_FX_EmissiveAdjustments}
  n2 = Power(Base=ComponentMask(None=TextureCoordinate()) {r=true}) {constExponent=0.25}
  n3 = Multiply(
   A=@n1.A
   B=Fresnel(ExponentIn=ScalarParameter() {parameterName=ExponentIn}, BaseReflectFractionIn=ScalarParameter() {parameterName=BaseReflectFractionIn})
  )
  n4 = MaterialFunctionCall(Float2=TextureCoordinate()) {materialFunction=&BreakOutFloat2Components}
  n5 = Multiply(A=@n4.R, B=ScalarParameter() {defaultValue=0.75, parameterName=Utile})
  n6 = ScalarParameter() {defaultValue=2, parameterName=UstartStretch}
  n7 = Panner(
   Coordinate=AppendVector(A=Power(Base=@n5, Exp=@n6), B=@n4.G)
   Time=Add(A=Time(), B=ScalarParameter() {parameterName=TimeOffset})
   Speed=AppendVector(A=ScalarParameter() {defaultValue=-1, parameterName=Speed}, B=Constant())
  ) {speedX=-9}
  n8 = TextureSampleParameter2D(UVs=@n7) {parameterName=TrailScrollTexture, texture=/Game/Weapons/FORT_Melee/Effects/LaserSwordFounders/T_LaserSwordTrailNoise}
  n9 = TextureSampleParameter2D(UVs=@n7) {parameterName=TrailScrollTexture_1, texture=&T_LaserSwordTrailNoise}
  Multiply(A=Multiply(A=OneMinus(None=@n2), B=Clamp(None=@n3)), B=Power(Base=Add(A=@n8.G, B=@n9.G)) {constExponent=0.5})
 ]
}
