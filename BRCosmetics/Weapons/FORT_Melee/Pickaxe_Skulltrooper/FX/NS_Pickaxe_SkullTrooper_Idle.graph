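# Niagara idle effect for the Skull Trooper pickaxe, emitting a slow trickle of translucent sprite particles from sockets on the weapon.
# A single emitter spawns at an offset from a socket reader location, pushes particles along negative Y, then slows them with drag.
# Sprites grow, animate their sub-image index along a curve and fade in and out.
# User.VariantID indexes a two-color array to pick the blue or green tint.
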NiagaraSystem /BRCosmetics/Weapons/FORT_Melee/Pickaxe_Skulltrooper/FX/NS_Pickaxe_SkullTrooper_Idle
dependencies {
 usedDataInterfaces [&NiagaraDataInterfaceArrayColor, &NiagaraDataInterfaceCurve, &NiagaraDataInterfaceSocketReader]
 usedDynamicInputs []
 usedModules [
  &self:SystemSpawnScript.SystemScriptSource.SystemScriptGraph.NiagaraNodeAssignment_0.SetVariables_#1
  &SystemState
 ]
}
emitters {Minimal {
  emitterUpdateScript [
   &EmitterState()
   &SpawnRate(SpawnRate=3.5, Use Spawn Probability=false)
  ]
  particleSpawnScript [
   &InitializeParticle(
    Color=&SelectLinearColorFromArray(
     Array Sampling Mode=::NewEnumerator1
     Direct Array Index=&MakeIntFromFloat(Float=$User.VariantID)
     Color Selection Array=DataInterface() {propertyValues {ColorData [<rgba 0.2091 0.8295 1 0.3>, <rgba 0.4983 1 0.6519 0.3>]}}
     Random Seed=?
    )
    Lifetime Max=2.5
    Lifetime Min=1.25
    Lifetime Mode=::NewEnumerator1
    Sprite Rotation Mode=::NewEnumerator2
    Sprite Size Mode=::NewEnumerator1
    Uniform Sprite Size Max=40
    Uniform Sprite Size Min=25
    UsePositionOffset=false
    Write Material Random=true
   )
   &AddVelocity(Velocity=&RandomRangeVector(Minimum=<0 -50 0>, Maximum=<0 -150 0>))
   &SocketLocation(
    SamplePosOffset=true
    Sampled Position Offset=<0 26 0>
    Socket Reader=$System.Socket Reader
    Socket Sampling Mode=::NewEnumerator3
    VelocitySpeedLimit=false
   )
  ]
  particleUpdateScript [
   &ParticleState()
   &ScaleSpriteSize(Uniform Curve Sprite Scale=DataInterface() {propertyValues {Curve {keys [{interpMode=RCIM_Cubic, leaveTangent=2.8675, tangentMode=RCTM_Break, time=0, value=0.3}, {tangentMode=RCTM_User, time=1, value=1.4}]}}})
   /SetVariables(Particles.SubImageIndex=&FloatFromCurve(
     FloatCurve=DataInterface() {propertyValues {Curve {keys [
         {arriveTangent=1.4707, interpMode=RCIM_Cubic, leaveTangent=1.4651, tangentMode=RCTM_Break, time=0, value=0}
         {arriveTangent=1.4707, interpMode=RCIM_Cubic, tangentMode=RCTM_Break, time=1, value=1}
        ]}}}
     Scale Curve=40
    )) {isSetParameters=true}
   &ScaleColor(
    Scale Alpha=&FloatFromCurve(
     FloatCurve=DataInterface() {propertyValues {Curve {keys [
         {interpMode=RCIM_Cubic, leaveTangent=11.0386, tangentMode=RCTM_Break, time=0, value=0}
         {interpMode=RCIM_Cubic, leaveTangent=-1.8558, tangentMode=RCTM_Break, time=0.2, value=1}
         <at 1 0>
        ]}}}
     Scale Curve=&RandomRangeFloat(Minimum=1, Maximum=1.3)
    )
    ScaleA=true
    ScaleRGB=true
   )
   &Drag(Drag=3.5, Use Linear Drag=true, Use Rotational Drag=true)
   &SolveForcesAndVelocity(Clamp Velocity=false, Limit Acceleration=false)
   &CameraOffset(Camera Offset Amount=15)
  ]
  rendererClasses [&NiagaraSpriteRendererProperties]
  renderers [{rendererClass=&NiagaraSpriteRendererProperties}]
 }}
summary {
 systemName=NS_Pickaxe_SkullTrooper_Idle
 emitters [{emitterName=Minimal, rendererClasses [&NiagaraSpriteRendererProperties]}]
 userVariables [User.VariantID {type=NiagaraFloat, defaultValue=0, description=""}]
}
